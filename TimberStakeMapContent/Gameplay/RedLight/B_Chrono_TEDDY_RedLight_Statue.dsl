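; Red Light, Green Light statue: punishes players moving inside its detection box during red light with gameplay effects and a caught cue.
; The server loops randomized green and red waits, replicates timestamped state, and runs a repeating velocity check on overlapping pawns after the voice-and-telegraph delay.
; Clients chain two randomly chosen voice lines, then recolor the eye glow and turn the head via timeline, trimming audio and rotation for replication latency.
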(event Custom|RotateRobotOverTime
  (bind _rlglcurrentstate_isredlightstate_5_8434c50a411cd3f4ff8b2b8b80277c1c (Utilities|Struct|BreakFRedLightState (Variables|GetRLGLCurrentState_IsRedLightState_5_8434C50A411CD3F4FF8B2B8B80277C1C)))
  ;[0.1273 0.01095 0.15 0.5] Handle Robot Rotation on clients
  (Variables|Detection|SetCachedRLGLState _rlglcurrentstate_isredlightstate_5_8434c50a411cd3f4ff8b2b8b80277c1c)
  (bind (_direction _yawcurve) (Timeline|RobotRotationTimeline PlayFromStart)
    (:Update
      (Transformation|SetRelativeRotation (Variables|Default|GetHead) (Math|Rotator|MakeRotator 0.0 0.0 (Math|Float|SelectFloat (Math|Float|Lerp 1.8e+02 0.0 (Math|Float|Min(Float) (/ _yawcurve (Math|Float|Max(Float) (- (Math|Float|SelectFloat (ScalableFloat|GetValueatLevel (Variables|BalancingData|GetTelegraphForTurntoRed)) (ScalableFloat|GetValueatLevel (Variables|BalancingData|GetTelegraphForTurntoGreen)) (Variables|Detection|GetCachedRLGLState)) (Variables|GameplayReplicatedLogic|GetClientStateTransitionLatency)) 0.1)) 1.0)) (Math|Float|Lerp 0.0 1.8e+02 (Math|Float|Min(Float) (/ _yawcurve (Math|Float|Max(Float) (- (Math|Float|SelectFloat (ScalableFloat|GetValueatLevel (Variables|BalancingData|GetTelegraphForTurntoRed)) (ScalableFloat|GetValueatLevel (Variables|BalancingData|GetTelegraphForTurntoGreen)) (Variables|Detection|GetCachedRLGLState)) (Variables|GameplayReplicatedLogic|GetClientStateTransitionLatency)) 0.1)) 1.0)) (Variables|Detection|GetCachedRLGLState))) false false))))

(event Custom|CheckPlayersInsideRedLightZone
  ;[0.02288 0.01108 0.15 0.5] Red Light Check Players Logic
  (Variables|Detection|SetDo%20Caught%20GC true)
  (for _array_element (Variables|Teleport|GetPlayerArray)
    (Utilities|IsValid _array_element
      (:Is%20Valid
        ;; Player Is Moving
        (if (> (Math|Vector|VectorLengthSquared (Transformation|GetVelocity _array_element)) 0)
          (Utilities|FlowControl|Sequence
            (:then_0
              ;; Play Caught GC. Only once per iteration
              (if (Variables|Detection|GetDo%20Caught%20GC)
                (Variables|Detection|SetDo%20Caught%20GC false)
                (GameplayCue|ExecuteGameplayCue self (Variables|Detection|GetCaught%20Gameplay%20Cue%20Tag))
                (Components|Activation|Activate (Variables|Default|GetNS_S17_Mindwipe_Flash) false)))
            (:then_1
              ;; Apply Damage GE to target
              (Fortnite|ApplyGameplayEffectsToActor _array_element (Variables|Detection|GetEffects%20to%20Apply) (Transformation|GetActorLocation)))))))))

(event Custom|StartRLGLTransition
  (params (LoopingAudioTime :type Integer)))

(event Custom|RLGLLogic
  ;[0.01642 0.15 0.07222 0.5] Main Red Light Green Light Logic
  (if (Networking|HasAuthority)
    ;; Randomize rotate duration
    (CallFunction|SetWaitDuration)
    (Utilities|FlowControl|Delay (Math|Random|RandomFloatinRange (ScalableFloat|GetValueatLevel (Variables|BalancingData|GetGreenVariableWaitTime)) (ScalableFloat|GetValueatLevel (Variables|BalancingData|GetGreenVariableWaitTime) 1.0)))
    ;; Trigger Red Light State
    (Variables|SetRLGLCurrentState_IsRedLightState_5_8434C50A411CD3F4FF8B2B8B80277C1C (Utilities|Struct|MakeFRedLightState true (Fortnite|Sync|Time|GetServerWorldTimeSeconds) (Variables|GameplayReplicatedLogic|GetVOIndexFirstPart) (Variables|GameplayReplicatedLogic|GetVOIndexSecondPart)))
    (Utilities|FlowControl|Delay (CallFunction|ServerCalculateRedLightTransitionDuration))
    ;; Check players inside red light zone
    (bind _returnvalue (Utilities|Time|SetTimerbyEvent (EventDispatchers|CreateEvent :!event Custom|CheckPlayersInsideRedLightZone) (Variables|Detection|GetCheckPlayersInRedZoneTime) true))
    (Variables|Detection|SetCheckPlayerInsideRedLightZoneHandle _returnvalue)
    (Utilities|FlowControl|Delay (Math|Random|RandomFloatinRange (ScalableFloat|GetValueatLevel (Variables|BalancingData|GetRedVariableWaitTime)) (ScalableFloat|GetValueatLevel (Variables|BalancingData|GetRedVariableWaitTime) 1.0)))
    ;; Trigger Green Light State
    (Variables|SetRLGLCurrentState_IsRedLightState_5_8434C50A411CD3F4FF8B2B8B80277C1C (Utilities|Struct|MakeFRedLightState false (Fortnite|Sync|Time|GetServerWorldTimeSeconds) -1 -1))
    (Utilities|FlowControl|Delay (CallFunction|CalculateGreenLightTransitionDuration))
    ;; Clear check players timer
    (Utilities|Time|ClearandInvalidateTimerbyHandle (Variables|Detection|GetCheckPlayerInsideRedLightZoneHandle))
    ;; Call again to trigger a new RLGL process
    (CallFunction|RLGLLogic)))

(event OnComponentBeginOverlap(DetectionBox)
  (params (OverlappedComponent :type Primitive%20Component%20Object%20Reference) (OtherActor :type Actor%20Object%20Reference) (OtherComp :type Primitive%20Component%20Object%20Reference) (OtherBodyIndex :type Integer) (bFromSweep :type Boolean) (SweepResult :type Hit%20Result%20Structure%20%28by%20ref%29))
  ;[1 1 1] Overlaps
  (bind _asfort_player_pawn (Utilities|Casting|CastToFortPlayerPawn OtherActor))
  (Utilities|Array|AddUnique (Variables|Teleport|GetPlayerArray) _asfort_player_pawn))

(event EventBeginPlay
  ;[0.15 0.006036 2.3e-05 0.5] Begin Play
  ;; We want to force an update state as the default value is false and won't call the OnRep otherwise
  (CallFunction|UpdateRLGLStateAfterSound)
  ;; Only the server is going to keep track of the rotation. Clients will receive rotation updates through ReplicateMovement
  (if (Networking|HasAuthority)
    ;; Start Red light green light logic.
    (CallFunction|RLGLLogic)))

(event OnComponentEndOverlap(DetectionBox)
  (params (OverlappedComponent :type Primitive%20Component%20Object%20Reference) (OtherActor :type Actor%20Object%20Reference) (OtherComp :type Primitive%20Component%20Object%20Reference) (OtherBodyIndex :type Integer))
  ;[1 1 1] Overlaps
  (bind _asfort_player_pawn (Utilities|Casting|CastToFortPlayerPawn OtherActor))
  (Utilities|Array|RemoveItem (Variables|Teleport|GetPlayerArray) _asfort_player_pawn))

(fn CalculateGreenLightTransitionDuration
  (return (ScalableFloat|GetValueatLevel (Variables|BalancingData|GetTelegraphForTurntoGreen))))

(fn GetFirstVOSound
  (bind (_rlglcurrentstate_isredlightstate_5_8434c50a411cd3f4ff8b2b8b80277c1c _rlglcurrentstate_timestamp_4_eae0a0df40e7dfc53ead6b8835e680e8 _rlglcurrentstate_voindexfirstpart_14_332334bb4fb0ee7057ce5dad24b69e72) (Utilities|Struct|BreakFRedLightState (Variables|GetRLGLCurrentState_IsRedLightState_5_8434C50A411CD3F4FF8B2B8B80277C1C)))
  (return (Utilities|Select :Index _rlglcurrentstate_voindexfirstpart_14_332334bb4fb0ee7057ce5dad24b69e72 :Option%200 (Variables|Default|GetFirstVO_Short) :Option%201 (Variables|Default|GetFirstVO_Medium) :Option%202 (Variables|Default|GetFirstVO_Long) :Option%203 (Variables|Default|GetFirstVO_ExtraLong))))

(fn GetSecondVOSound
  (bind (_rlglcurrentstate_isredlightstate_5_8434c50a411cd3f4ff8b2b8b80277c1c _rlglcurrentstate_timestamp_4_eae0a0df40e7dfc53ead6b8835e680e8 _rlglcurrentstate_voindexfirstpart_14_332334bb4fb0ee7057ce5dad24b69e72 _rlglcurrentstate_voindexsecondpart_15_ef56f9b641f1bf8e423f258aa6c76f46) (Utilities|Struct|BreakFRedLightState (Variables|GetRLGLCurrentState_IsRedLightState_5_8434C50A411CD3F4FF8B2B8B80277C1C)))
  (return (Utilities|Select :Index _rlglcurrentstate_voindexsecondpart_15_ef56f9b641f1bf8e423f258aa6c76f46 :Option%200 (Variables|Default|GetSecondVO_Short) :Option%201 (Variables|Default|GetSecondVO_Medium) :Option%202 (Variables|Default|GetSecondVO_Long))))

(fn OnRep_RLGLCurrentState
  (bind (_rlglcurrentstate_isredlightstate_5_8434c50a411cd3f4ff8b2b8b80277c1c _rlglcurrentstate_timestamp_4_eae0a0df40e7dfc53ead6b8835e680e8) (Utilities|Struct|BreakFRedLightState (Variables|GetRLGLCurrentState_IsRedLightState_5_8434C50A411CD3F4FF8B2B8B80277C1C)))
  (Variables|GameplayReplicatedLogic|SetClientState "NewEnumerator3")
  (Variables|GameplayReplicatedLogic|SetClientStateTransitionLatency (Math|Float|Max(Float) (- (Fortnite|Sync|Time|GetServerWorldTimeSeconds) _rlglcurrentstate_timestamp_4_eae0a0df40e7dfc53ead6b8835e680e8)))
  (CallFunction|UpdateClientRLGLState))

(fn PlayRedStateVO
  (params (VOToPlay :type Audio%20Component%20Object%20Reference) (Latency :type Float%20%28double-precision%29))
  (bind _latency (Variables|GetLatency))
  (bind _votoplay (Variables|GetVOToPlay))
  (bind _sound (Variables|/Script/Engine.AudioComponent|GetSound _votoplay))
  (bind _duration (Variables|/Script/Engine.SoundBase|GetDuration _sound))
  (Variables|GameplayReplicatedLogic|SetClientStateTransitionLatency (Math|Float|Max(Float) (- (Variables|GameplayReplicatedLogic|GetClientStateTransitionLatency) (Variables|/Script/Engine.SoundBase|GetDuration (Variables|/Script/Engine.AudioComponent|GetSound (Variables|GetVOToPlay))))))
  ; Start sound at the right time taking into account latency
  (if (> _duration _latency)
    (Audio|Components|Audio|Play (Variables|GetVOToPlay) (Variables|GetLatency))
    (return (- _duration _latency))
    (else
      (return 0.0))))

(fn PlaySecondVO
  (CallFunction|UpdateClientRLGLState))

(fn ServerCalculateRedLightTransitionDuration
  (return (Utilities|Operators|Add (ScalableFloat|GetValueatLevel (Variables|BalancingData|GetTelegraphForTurntoRed)) (Variables|/Script/Engine.SoundBase|GetDuration (Variables|/Script/Engine.AudioComponent|GetSound (CallFunction|GetFirstVOSound))) (Variables|/Script/Engine.SoundBase|GetDuration (Variables|/Script/Engine.AudioComponent|GetSound (CallFunction|GetSecondVOSound))))))

(fn SetWaitDuration
  (Variables|GameplayReplicatedLogic|SetVOIndexFirstPart (Math|Random|RandomIntegerinRange 0 3))
  (Variables|GameplayReplicatedLogic|SetVOIndexSecondPart (Math|Random|RandomIntegerinRange 0 2)))

(fn UpdateClientRLGLState
  (if (not (Networking|IsDedicatedServer))
    (Utilities|Time|ClearandInvalidateTimerbyHandle (Variables|Detection|GetRLGLSoundPlayingTimer))
    (if (Utilities|Struct|BreakFRedLightState (Variables|GameplayReplicatedLogic|GetRLGLCurrentState))
      (switch Utilities|FlowControl|Switch|SwitchonE_RLGLClientState (Variables|GameplayReplicatedLogic|GetClientState)
        (:NewEnumerator3
          _)
        (:NewEnumerator0
          _)
        (:NewEnumerator1
          (Variables|GameplayReplicatedLogic|SetClientState "NewEnumerator2")
          (CallFunction|UpdateRLGLStateAfterSound)))
      ; Select sound based on state and duration index from server
      (bind _soundduration (CallFunction|PlayRedStateVO (select (Utilities|Enum|Equal(Enum) (Variables|GameplayReplicatedLogic|GetClientState) "NewEnumerator3") (CallFunction|GetFirstVOSound) (CallFunction|GetSecondVOSound)) (Variables|GameplayReplicatedLogic|GetClientStateTransitionLatency)))
      (Variables|GameplayReplicatedLogic|SetClientState (select (Utilities|Enum|Equal(Enum) (Variables|GameplayReplicatedLogic|GetClientState) "NewEnumerator3") "NewEnumerator0" "NewEnumerator1"))
      (if (> _soundduration 0)
        ; Wait state duration
        (bind _returnvalue (Utilities|Time|SetTimerbyEvent (EventDispatchers|CreateEvent :!event UpdateClientRLGLState) _soundduration))
        (Variables|Detection|SetRLGLSoundPlayingTimer _returnvalue)
        (elif (Utilities|Enum|Equal(Enum) (Variables|GameplayReplicatedLogic|GetClientState) "NewEnumerator0")
          ; Trigger second VO because first was skipped due to latency
          (CallFunction|UpdateClientRLGLState)
          (else
            _))))))

(fn UpdateRLGLStateAfterSound
  (bind _ns_redllight_eyeglows (Variables|Default|GetNS_RedlLight_EyeGlows))
  (bind _rlglcurrentstate_isredlightstate_5_8434c50a411cd3f4ff8b2b8b80277c1c (Utilities|Struct|BreakFRedLightState (Variables|GetRLGLCurrentState_IsRedLightState_5_8434C50A411CD3F4FF8B2B8B80277C1C)))
  (if (not (Networking|IsDedicatedServer))
    (Utilities|FlowControl|Sequence
      (:then_0
        (if _rlglcurrentstate_isredlightstate_5_8434c50a411cd3f4ff8b2b8b80277c1c
          ;[0.15 3.2e-05 0.008604 0.5] Red Light Logic
          (Niagara|SetNiagaraVariable(LinearColor) (Variables|Default|GetNS_RedlLight_EyeGlows) "Color" (Utilities|Struct|MakeLinearColor 1.0 0.0 0.0 1.0))
          (else
            ;[0.04452 0.15 0.01765 0.5] Green Light Logic
            (Niagara|SetNiagaraVariable(LinearColor) _ns_redllight_eyeglows "Color" (Utilities|Struct|MakeLinearColor 0.0 1.0 0.0 1.0))
            (Niagara|SetNiagaraVariable(Float) _ns_redllight_eyeglows "Green Light Radius" (Variables|Visuals|GetGreenLightRadius))
            (Niagara|SetNiagaraVariable(Float) _ns_redllight_eyeglows "Green Light Offset" (Variables|Visuals|GetGreenLightOffset))
            (Niagara|SetNiagaraVariable(Float) _ns_redllight_eyeglows "Green Light Intensity" (Variables|Visuals|GetGreenLightIntensity)))))
      (:then_1
        ; Rotate Robot
        (CallFunction|RotateRobotOverTime)))))

(fn ConstructionScript)
